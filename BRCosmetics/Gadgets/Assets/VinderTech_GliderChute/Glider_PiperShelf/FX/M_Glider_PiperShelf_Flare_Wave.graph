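# Translucent unlit Niagara sprite material for a flare wave effect on the PiperShelf glider.
# A vertically flipped grayscale wave texture, borrowed from the Monarch glider, is sharpened by an Exponent parameter.
# A powered camera-to-vertex-normal dot product fades the mask as the surface turns away from view.
# The saturated mask multiplies particle color, passed through the MF_FX_EmissiveAdjustments function for emissive, and particle alpha for transparency.
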Material /BRCosmetics/Gadgets/Assets/VinderTech_GliderChute/Glider_PiperShelf/FX/M_Glider_PiperShelf_Flare_Wave
graph {
 properties {bUsedWithNiagaraSprites=true, blendMode=BLEND_Translucent, pixelDepthOffsetMode=PDOM_Legacy, shadingModel=MSM_Unlit, usedShadingModels=Unlit}
 nodes [
  n1 = ParticleColor()
  n2 = TextureSample(UVs=TextureCoordinate() {vTiling=-1}) {
   samplerType=SAMPLERTYPE_LinearGrayscale
   texture=/BRCosmetics/Gadgets/Assets/VinderTech_GliderChute/Glider_Monarch_Female/FX/Textures/T_Glider_Monarch_Flare_Wave
  }
  n3 = Power(Base=@n2.R, Exp=ScalarParameter() {defaultValue=2, parameterName=Exponent})
  n4 = Power(Base=DotProduct(A=CameraVectorWS(), B=VertexNormalWS())) {constExponent=6}
  n5 = Multiply(A=Saturate(None=Multiply(A=@n3, B=@n4)))
  MaterialFunctionCall(Emissive=Multiply(A=@n1.RGB, B=@n5)) {materialFunction=&MF_FX_EmissiveAdjustments}
  Multiply(A=@n1.A, B=@n5)
 ]
}
