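# Unlit, two-sided additive glow material for Niagara mesh particles, built for the PiperShelf glider's glow mesh.
# Particle color RGB is added to a multiplied copy of itself and passed through the shared FX emissive adjustments function.
# A fade mask multiplies particle alpha by a saturated power curve of the UV V coordinate, giving a procedural gradient without textures.
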Material /BRCosmetics/Gadgets/Assets/VinderTech_GliderChute/Glider_PiperShelf/FX/M_PiperShelf_Glow
graph {
 properties {
  bAutomaticallySetUsageInEditor=false
  bUsedWithNiagaraMeshParticles=true
  blendMode=BLEND_Additive
  previewMesh=&/SM_Glider_PiperShelf_Glow_01
  shadingModel=MSM_Unlit
  twoSided=true
  usedShadingModels=Unlit
 }
 nodes [
  n1 = ParticleColor()
  MaterialFunctionCall(Emissive=Add(A=@n1.RGB, B=Multiply(A=@n1.RGB))) {materialFunction=&MF_FX_EmissiveAdjustments}
  n2 = ComponentMask(None=TextureCoordinate()) {g=true}
  Multiply(A=@n1.A, B=Saturate(None=Power(Base=@n2)))
 ]
}
