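# Translucent unlit Niagara ribbon material for the PiperShelf glider chute FX.
# Particle color feeds MF_FX_EmissiveAdjustments for emissive adjustment.
# UV math builds procedural masks without textures: a powered falloff along U, a quick ramp-in at the U start, and a sine-based soft profile across V.
# The U falloff is also multiplied by particle alpha.
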Material /BRCosmetics/Gadgets/Assets/VinderTech_GliderChute/Glider_PiperShelf/FX/M_PiperShelf_Ribbon
graph {
 properties {
  bAutomaticallySetUsageInEditor=false
  bUsedWithNiagaraRibbons=true
  blendMode=BLEND_Translucent
  shadingModel=MSM_Unlit
  usedShadingModels=Unlit
 }
 nodes [
  n1 = ParticleColor()
  MaterialFunctionCall(Emissive=@n1.RGB) {materialFunction=&MF_FX_EmissiveAdjustments}
  n2 = TextureCoordinate()
  n3 = OneMinus(None=ComponentMask(None=@n2) {r=true})
  n4 = Saturate(None=Multiply(A=Power(Base=@n3) {constExponent=1.6}))
  Multiply(A=@n1.A, B=@n4)
  n5 = ComponentMask(None=@n2) {r=true}
  Multiply(A=@n4, B=Saturate(None=Multiply(A=@n5) {constB=17}))
  Power(Base=Sine(None=ComponentMask(None=@n2) {g=true}) {period=2}) {constExponent=0.4}
 ]
}
